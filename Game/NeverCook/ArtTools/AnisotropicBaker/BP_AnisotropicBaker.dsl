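; Art-tool actor that bakes an anisotropic tangent map from a grain direction and normal map into a render target.
; Construction script assigns meshes, creates bake and test material instances, previews them on the chosen material slot and limits scene capture to this actor.
; Bake hides static mesh slots with a placeholder material, then toggles the material's Bake parameter around a one-shot capture, apparently rendering the mesh flattened.
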(event EventBeginPlay)

(event Collision|EventActorBeginOverlap
  (params (OtherActor :type Actor%20Object%20Reference)))

(event EventTick
  (params (DeltaSeconds :type Float%20%28single-precision%29)))

(event Custom|Bake
  (for _index (range 51)
    (Rendering|Material|SetMaterial (Variables|Default|GetMeshToBakeSM) _index "/Game/NeverCook/ArtTools/AnisotropicBaker/RTT_HiddenElement.RTT_HiddenElement"))
  ;[1 1 1] Reset to baking material on capture
  (Rendering|Material|SetMaterial (Variables|Default|GetMeshToBake) (Variables|Default|GetMat%20ID%20ElementToBake) (Variables|Default|GetAnisoMaterial))
  ;[1 1 1] Capture RenderTarget
  (Rendering|ClearRenderTarget2D (Variables|Default|GetCaptureTarget(Dont Edit)))
  (Rendering|Material|SetScalarParameterValue (Variables|Default|GetAnisoMaterial) "Bake" 1.0 :!class /Script/Engine.MaterialInstanceDynamic)
  (Rendering|SceneCapture|CaptureScene (Variables|Default|GetCaptureCamera) :!class /Script/Engine.SceneCaptureComponent2D)
  (Rendering|Material|SetScalarParameterValue (Variables|Default|GetAnisoMaterial) "Bake" 0.0 :!class /Script/Engine.MaterialInstanceDynamic)
  ;[1 1 1] Show preview mat if tex valid
  (Utilities|IsValid (Variables|Default|GetTangentMapToTest)
    (:Is%20Valid
      (Rendering|Material|SetTextureParameterValue (Variables|Default|GetPostBakeTest) "TangentMap" (Variables|Default|GetTangentMapToTest)))))

(fn ConstructionScript
  (Transformation|SetActorLocation "0, 0, 0")
  (Transformation|SetWorldRotation (Variables|Default|GetTangentDirectionVisualizer) (Math|Rotator|MakeRotfromX (Variables|Default|GetGrain%20Direction)) false false)
  (Variables|/Script/Engine.SceneCaptureComponent2D|SetTextureTarget (Variables|Default|GetCaptureTarget(Dont Edit)) (Variables|Default|GetCaptureCamera))
  (Utilities|FlowControl|Sequence
    (:then_0
      ;[1 1 1] Set Meshes
      (Components|SkinnedMesh|SetSkinnedAssetandUpdate (Variables|Default|GetMeshToBake) (Variables|Default|GetSkeletalMesh) true)
      (Components|StaticMesh|SetStaticMesh (Variables|Default|GetMeshToBakeSM) (Variables|Default|GetStaticMesh))
      ;[1 1 1] Setup Materials
      (bind _returnvalue (Rendering|Material|CreateDynamicMaterialInstance "/Game/NeverCook/ArtTools/AnisotropicBaker/M_BakeAniso.M_BakeAniso" "None" "None" :!class /Script/Engine.KismetMaterialLibrary))
      (Variables|Default|SetAnisoMaterial _returnvalue)
      (bind _returnvalue_2 (Rendering|Material|CreateDynamicMaterialInstance "/Game/NeverCook/ArtTools/AnisotropicBaker/M_BakeTest.M_BakeTest" "None" "None" :!class /Script/Engine.KismetMaterialLibrary))
      (Variables|Default|SetPostBakeTest _returnvalue_2)
      (Rendering|Material|SetTextureParameterValue (Variables|Default|GetAnisoMaterial) "NormalMap" (Variables|Default|GetNormalMap))
      (Rendering|Material|SetScalarParameterValue (Variables|Default|GetAnisoMaterial) "FlattenNormals" (Variables|Default|GetFlattenNormalsForAnisoCalculation) :!class /Script/Engine.MaterialInstanceDynamic)
      (Rendering|Material|SetScalarParameterValue (Variables|Default|GetAnisoMaterial) "Bake" 0.0 :!class /Script/Engine.MaterialInstanceDynamic)
      (Rendering|Material|SetVectorParameterValue (Variables|Default|GetAnisoMaterial) "TangentDirection" (Math|Conversions|ToLinearColor(Vector) (Math|Vector|Normalize (Variables|Default|GetGrain%20Direction))) :!class /Script/Engine.MaterialInstanceDynamic)
      (Rendering|Material|SetMaterial (Variables|Default|GetMeshToBake) (Variables|Default|GetMat%20ID%20ElementToBake) (Variables|Default|GetAnisoMaterial))
      (Rendering|Material|SetScalarParameterValue (Variables|Default|GetAnisoMaterial) "Bake" (Math|Conversions|ToFloat(Boolean) (Variables|Default|GetPreviewFlattenedBake)) :!class /Script/Engine.MaterialInstanceDynamic)
      ;[1 1 1] Preview Tangent  map if provided
      (Utilities|IsValid (Variables|Default|GetTangentMapToTest)
        (:Is%20Valid
          (Rendering|Material|SetTextureParameterValue (Variables|Default|GetPostBakeTest) "TangentMap" (Variables|Default|GetTangentMapToTest))
          (Rendering|Material|SetMaterial (Variables|Default|GetMeshToBake) (Variables|Default|GetMat%20ID%20ElementToBake) (Variables|Default|GetPostBakeTest)))))
    (:then_1
      (Rendering|SceneCapture|ShowOnlyActorComponents (Variables|Default|GetCaptureCamera) self false))))
