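# Two-sided Default Lit art-tool material that bakes an anisotropy tangent direction into UV space, with a Bake parameter switching from preview to bake.
# The tangent is the cross product of TangentDirection and a pixel-to-vertex normal blend set by FlattenNormals, transformed to tangent space.
# For baking, it is sign-corrected and remapped to color, while a world-position delta flattens the mesh into its UV square.
# Anisotropy, roughness and normal map parameters support the lit preview.
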Material /Game/NeverCook/ArtTools/AnisotropicBaker/M_BakeAniso
graph {
 properties {bUsedWithMeshDeformer=true, bUsedWithSkeletalMesh=true, pixelDepthOffsetMode=PDOM_Legacy, twoSided=true, usedShadingModels=Default Lit}
 nodes [
  Constant() {r=1}
  n1 = TwoSidedSign()
  n2 = VectorParameter() {parameterName=TangentDirection, defaultValue=<rgba 0 0 1 0>}
  n3 = PixelNormalWS()
  n4 = VertexNormalWS()
  n5 = ScalarParameter() {parameterName=FlattenNormals}
  n6 = Transform(None=CrossProduct(A=Normalize(VectorInput=@n2.RGB), B=LinearInterpolate(A=@n3, B=@n4, Alpha=@n5))) {transformSourceType=TRANSFORMSOURCE_World, transformType=TRANSFORM_Tangent}
  Multiply(A=ConstantBiasScale(None=Multiply(A=@n1, B=@n6)), B=ScalarParameter() {parameterName=Bake})
  n7 = Frac(None=TextureCoordinate())
  n8 = AppendVector(A=Multiply(A=ConstantBiasScale(None=@n7) {bias=-0.5, scale=2}) {constB=0.5}, B=Constant())
  n9 = WorldPosition() {worldPositionShaderOffset=WPT_ExcludeAllShaderOffsets}
  Reroute(None=Multiply(A=Subtract(A=@n8, B=@n9.XYZ) {constA=0}, B=ScalarParameter() {parameterName=Bake}))
  n10 = ScalarParameter() {parameterName=Bake}
  Reroute(None=Multiply(A=Constant() {r=0.9}, B=OneMinus(None=@n10)))
  Reroute(None=@n6)
  ScalarParameter() {defaultValue=0.8, parameterName=Anisotropy}
  ScalarParameter() {defaultValue=0.7, parameterName=roughness}
  TextureSampleParameter2D() {parameterName=NormalMap, samplerType=SAMPLERTYPE_Normal, texture=&FlatNormalBlue}
 ]
}
