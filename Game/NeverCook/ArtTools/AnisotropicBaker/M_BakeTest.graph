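# M_BakeTest is a two-sided Default Lit test material in the AnisotropicBaker art tools, apparently for previewing baked anisotropy maps.
# A TangentMap parameter, defaulting to an anisotropy texture, feeds a lerp that remaps its RGB from 0..1 to a signed -1..1 vector.
# A NormalMap parameter defaults to flat normal, alongside fixed scalar constants.
# It supports skeletal meshes and mesh deformers.
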Material /Game/NeverCook/ArtTools/AnisotropicBaker/M_BakeTest
graph {
 properties {bUsedWithMeshDeformer=true, bUsedWithSkeletalMesh=true, pixelDepthOffsetMode=PDOM_Legacy, twoSided=true, usedShadingModels=Default Lit}
 nodes [
  Constant() {r=0.9}
  Constant() {r=1}
  Constant() {r=0.7}
  Constant() {r=0.8}
  LinearInterpolate(
   A=Constant3Vector() {constant=<rgba -1 -1 -1 0>}
   Alpha=TextureSampleParameter2D() {parameterName=TangentMap, samplerType=SAMPLERTYPE_LinearColor, texture=/Game/Global/Textures/T_Default_Aniso}.RGB
  )
  TextureSampleParameter2D() {parameterName=NormalMap, samplerType=SAMPLERTYPE_Normal, texture=&FlatNormalBlue}
 ]
}
