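# Unlit material in the AnisotropicBaker art tools, named as the hidden-element material for render-to-texture work.
# Its only graph logic subtracts a default-valued constant 3-vector from world position.
# No material output connection is shown, so that subtraction does not visibly drive the surface.
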Material /Game/NeverCook/ArtTools/AnisotropicBaker/RTT_HiddenElement
graph {
 properties {pixelDepthOffsetMode=PDOM_Legacy, shadingModel=MSM_Unlit, usedShadingModels=Unlit}
 nodes [Subtract(A=WorldPosition().XYZ, B=Constant3Vector())]
}
